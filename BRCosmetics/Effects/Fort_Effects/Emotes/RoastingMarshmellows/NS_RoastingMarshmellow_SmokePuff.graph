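# Niagara smoke puff effect for the roasting marshmallows emote, built from a single sprite emitter, PuffSmoke.
# Spawn rate ramps up along a curve over a short emitter loop, while the burst spawn module is disabled.
# Grey sprites drift mostly sideways under drag, grow, animate through SubUV frames and fade in then out, with a slight camera offset.
# The averaged owner scale multiplies velocity and the size curve, so the puff follows component scale.
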NiagaraSystem /BRCosmetics/Effects/Fort_Effects/Emotes/RoastingMarshmellows/NS_RoastingMarshmellow_SmokePuff
dependencies {
 usedDataInterfaces [&NiagaraDataInterfaceCurve, &NiagaraDataInterfaceColorCurve]
 usedDynamicInputs []
 usedModules [&SystemState]
}
emitters {PuffSmoke {
  emitterUpdateScript [
   &EmitterState(Life Cycle Mode=::NewEnumerator1, Loop Behavior=::NewEnumerator1, Loop Duration=0.75, UseLoopDelay=false)
   &SpawnBurst_Instantaneous(Spawn Count=3, Use Spawn Probability=false) {enabled=false}
   &SpawnRate(
    SpawnRate=&FloatFromCurve(
     FloatCurve=DataInterface() {propertyValues {Curve {keys [<at 0 3 RCIM_Cubic>, {arriveTangent=18.9675, interpMode=RCIM_Cubic, leaveTangent=18.9674, tangentMode=RCTM_User, time=1, value=9}]}}}
     CurveIndex=$Emitter.NormalizedLoopAge
    )
    Use Spawn Probability=false
   )
  ]
  particleSpawnScript [
   /Niagara/Modules/Spawn/Initialization/InitializeParticle(
    Lifetime=&UniformRangedFloat(Minimum=1, Maximum=1)
    Mesh Scale=<1.5 1.5 1.5>
    Module.Write Color=false
    Module.Write Lifetime=true
    Module.Write Mass=true
    Module.Write Material Random=true
    Module.Write Position=false
    Module.Write Scale=false
    Module.Write SpriteRotation=true
    Module.Write SpriteSize=true
    Position=&SimulationPosition()
    Sprite Rotation=&UniformRangedFloat(Minimum=0, Maximum=1)
    Sprite Size=&UniformRangedVector2D(
     Minimum=&Vector2DFromFloat(Value=30)
     Maximum=&Vector2DFromFloat(Value=50)
    )
   )
   &AddVelocity(Velocity=&Multiply_VectorByFloat(
     Vector=&UniformRangedVector(Minimum=<-20 -20 0>, Maximum=<20 20 5>)
     Float=&self:AverageVector(Vector:Vector3f=$Engine.Owner.Scale)
    ))
  ]
  particleUpdateScript [
   &ParticleState()
   &Drag(Drag=3, Use Linear Drag=true, Use Rotational Drag=false)
   &SolveForcesAndVelocity(Clamp Velocity=false, Limit Acceleration=false)
   &Color(Color=&ColorFromCurve(ColorCurve=DataInterface() {propertyValues {
       AlphaCurve {keys [<at -0.0023 0 RCIM_Cubic>,<at 0.1821 0.15>,<at 0.9977 -0.003 RCIM_Cubic>]}
       BlueCurve {keys [<at -0.0023 0.3854 RCIM_Cubic>,<at 0.9977 0.3854 RCIM_Cubic>]}
       GreenCurve {keys [<at -0.0023 0.3854 RCIM_Cubic>,<at 0.9977 0.3854 RCIM_Cubic>]}
       RedCurve {keys [<at -0.0023 0.3854 RCIM_Cubic>,<at 0.9977 0.3854 RCIM_Cubic>]}
      }}))
   /Niagara/Modules/Update/SubUV/SubUVAnimation(Number Of Frames=20)
   &ScaleSpriteSize(Scale Factor=&Vector2DFromFloat(Value=&FloatFromCurve(
      FloatCurve=DataInterface() {propertyValues {Curve {keys [<at 0 0.5 RCIM_Cubic>,<at 1 1 RCIM_Cubic>]}}}
      Scale Curve=&self:AverageVector(Vector:Vector3f=$Engine.Owner.Scale)
     )))
   &CameraOffset(Camera Offset Amount=10)
  ]
  rendererClasses [&NiagaraSpriteRendererProperties]
  renderers [{rendererClass=&NiagaraSpriteRendererProperties}]
 }}
summary {
 systemName=NS_RoastingMarshmellow_SmokePuff
 emitters [{emitterName=PuffSmoke, rendererClasses [&NiagaraSpriteRendererProperties]}]
 userVariables []
}
