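# Niagara system for the default Roasting Marshmallows emote, producing soft grey smoke that drifts upward.
# The active SmokeCardsL001 emitter spawns rotated sprites that accelerate upward, grow, darken and fade, with a camera offset.
# Two sprite renderers, a separately scaled DetailSmokeSize, random-start 64-frame SubUV animation and curve-driven dynamic material parameters shape the smoke.
# SmokeCardsL, a faster, shorter-lived, fainter variant without camera offset, is disabled.
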NiagaraSystem /BRCosmetics/Effects/Fort_Effects/Emotes/RoastingMarshmellows/NS_RoastingMarshmellows_Emote_Default
dependencies {
 usedDataInterfaces [&NiagaraDataInterfaceColorCurve, &NiagaraDataInterfaceCurve]
 usedDynamicInputs []
 usedModules [&SystemState]
}
emitters {
 SmokeCardsL {
  bEnabled=false
  emitterSpawnScript [&FN_PreserveSkydiveUserVariables() {module=PreserveUserVariables}]
  emitterUpdateScript [
   &EmitterState()
   &SpawnRate(SpawnRate=10, Use Spawn Probability=false)
  ]
  particleSpawnScript [
   /Niagara/Modules/Spawn/Initialization/InitializeParticle(
    Color=<rgba 0.4844 0.4844 0.4844 0.01>
    Lifetime=&UniformRangedFloat(Minimum=0.8, Maximum=1)
    Module.Write Color=true
    Module.Write Lifetime=true
    Module.Write Mass=false
    Module.Write Material Random=true
    Module.Write Position=false
    Module.Write Scale=false
    Module.Write SpriteRotation=true
    Module.Write SpriteSize=true
    Sprite Rotation=&UniformRangedFloat(Minimum=-360, Maximum=360)
    Sprite Size=&Vector2DFromFloat(Value=&UniformRangedFloat(Minimum=5, Maximum=6))
   )
   /SetVariables(Particles.RandomStartingFrame:NiagaraFloat=&UniformRangedFloat(Minimum=0, Maximum=63)) {isSetParameters=true, module=SetVariables_#1}
   /SetVariables(Particles.DetailSmokeSize:Vector2f=&Vector2DFromFloat(Value=&UniformRangedFloat(Minimum=10, Maximum=15))) {isSetParameters=true, module=SetVariables_#2}
   &AddVelocity(Velocity=&UniformRangedVector(Minimum=<0 0 5>, Maximum=<1 1 10>))
  ]
  particleUpdateScript [
   &ParticleState()
   &AccelerationForce(Acceleration=<0 0 50>)
   &ScaleColor(
    Scale Alpha=&FloatFromCurve(FloatCurve=DataInterface() {propertyValues {Curve {keys [<at 0.25 1.25 RCIM_Cubic>,<at 0.9 0 RCIM_Cubic>]}}})
    Scale RGB=&MakeVectorFromLinearColorRGB(Color=&ColorFromCurve(ColorCurve=DataInterface() {propertyValues {
        AlphaCurve {keys []}
        BlueCurve {keys [<at 0.0469 1 RCIM_Cubic>,<at 0.256 0.651 RCIM_Cubic>]}
        GreenCurve {keys [<at 0.0469 1 RCIM_Cubic>,<at 0.256 0.651 RCIM_Cubic>]}
        RedCurve {keys [<at 0.0469 1 RCIM_Cubic>,<at 0.256 0.651 RCIM_Cubic>]}
       }}))
    ScaleA=true
    ScaleRGB=true
   )
   &SolveForcesAndVelocity(Clamp Velocity=false, Limit Acceleration=false)
   &ScaleSpriteSize(Scale Factor=&Vector2DFromFloat(Value=&FloatFromCurve(FloatCurve=DataInterface() {propertyValues {Curve {keys [<at 0 0.5 RCIM_Cubic>,<at 1 3>]}}})))
   /SetVariables(Particles.DetailSmokeSize:Vector2f=&Multiply_Vector2(
     A=$Particles.Initial.DetailSmokeSize
     B=&Vector2DFromFloat(Value=&FloatFromCurve(FloatCurve=DataInterface() {propertyValues {Curve {keys [<at 0 0.5 RCIM_Cubic>,<at 1 2>]}}}))
    )) {isSetParameters=true, module=SetVariables_#3}
   /Niagara/Modules/Update/SubUV/SubUVAnimation(Number Of Frames=64, SubUV Lookup Index=&Add_Float(A=$Particles.NormalizedAge, B=$Particles.RandomStartingFrame))
   &DynamicMaterialParameters(
    Index 0 Param 1=&FloatFromCurve(FloatCurve=DataInterface() {propertyValues {Curve {keys [<at 0.3 1 RCIM_Cubic>,<at 1 -0.1>]}}})
    Index 0 Param 2=&FloatFromCurve(FloatCurve=DataInterface() {propertyValues {Curve {keys [<at 0 1 RCIM_Cubic>,<at 0.8 8.8 RCIM_Cubic>]}}})
    Index 0 Param 4=&FloatFromCurve(FloatCurve=DataInterface() {propertyValues {Curve {keys [<at 0.3 0>,<at 0.5 0.5>]}}})
    Module.Param0WriteEnabled=true
    Module.Param1WriteEnabled=true
    Module.Param2WriteEnabled=true
    Module.Param3WriteEnabled=true
   )
  ]
  rendererClasses [&NiagaraSpriteRendererProperties]
  renderers [
   {rendererClass=&NiagaraSpriteRendererProperties}
   {rendererClass=&NiagaraSpriteRendererProperties, rendererIndex=1}
  ]
 }
 SmokeCardsL001 {
  emitterSpawnScript [&FN_PreserveSkydiveUserVariables() {module=PreserveUserVariables}]
  emitterUpdateScript [
   &EmitterState(Life Cycle Mode=::NewEnumerator1, Loop Duration=0, UseLoopDelay=false)
   &SpawnRate(SpawnRate=10, Use Spawn Probability=false)
  ]
  particleSpawnScript [
   &InitializeParticle(
    Color=<rgba 0.4844 0.4844 0.4844 0.05>
    Lifetime=&UniformRangedFloat(Minimum=1.4, Maximum=1.7)
    Module.Write Color=true
    Module.Write Lifetime=true
    Module.Write Mass=false
    Module.Write Material Random=true
    Module.Write Position=false
    Module.Write Scale=false
    Module.Write SpriteRotation=true
    Module.Write SpriteSize=true
    Sprite Rotation=&UniformRangedFloat(Minimum=-360, Maximum=360)
    Sprite Size=&Vector2DFromFloat(Value=&UniformRangedFloat(Minimum=10, Maximum=15))
   )
   /SetVariables(Particles.RandomStartingFrame:NiagaraFloat=&UniformRangedFloat(Minimum=0, Maximum=63)) {isSetParameters=true, module=SetVariables_#1}
   /SetVariables(Particles.DetailSmokeSize:Vector2f=&Vector2DFromFloat(Value=&UniformRangedFloat(Minimum=10, Maximum=15))) {isSetParameters=true, module=SetVariables_#2}
   &AddVelocity(Velocity=&UniformRangedVector(Minimum=<0 0 2.5>, Maximum=<1 1 5>))
  ]
  particleUpdateScript [
   &ParticleState()
   &AccelerationForce(Acceleration=<0 0 20>)
   &ScaleColor(
    Scale Alpha=&FloatFromCurve(FloatCurve=DataInterface() {propertyValues {Curve {keys [<at 0.25 1.25 RCIM_Cubic>,<at 0.9 0 RCIM_Cubic>]}}})
    Scale RGB=&MakeVectorFromLinearColorRGB(Color=&ColorFromCurve(ColorCurve=DataInterface() {propertyValues {
        AlphaCurve {keys []}
        BlueCurve {keys [<at 0.0469 1 RCIM_Cubic>,<at 0.256 0.651 RCIM_Cubic>]}
        GreenCurve {keys [<at 0.0469 1 RCIM_Cubic>,<at 0.256 0.651 RCIM_Cubic>]}
        RedCurve {keys [<at 0.0469 1 RCIM_Cubic>,<at 0.256 0.651 RCIM_Cubic>]}
       }}))
    ScaleA=true
    ScaleRGB=true
   )
   &SolveForcesAndVelocity(Clamp Velocity=false, Limit Acceleration=false)
   &ScaleSpriteSize(Scale Factor=&Vector2DFromFloat(Value=&FloatFromCurve(FloatCurve=DataInterface() {propertyValues {Curve {keys [<at 0 0.5 RCIM_Cubic>,<at 1 3>]}}})))
   /SetVariables(Particles.DetailSmokeSize:Vector2f=&Multiply_Vector2(
     A=$Particles.Initial.DetailSmokeSize
     B=&Vector2DFromFloat(Value=&FloatFromCurve(FloatCurve=DataInterface() {propertyValues {Curve {keys [<at 0 0.5 RCIM_Cubic>,<at 1 2>]}}}))
    )) {isSetParameters=true, module=SetVariables_#3}
   &SubUVAnimation(Number Of Frames=64, SubUV Lookup Index=&Add_Float(A=$Particles.NormalizedAge, B=$Particles.RandomStartingFrame))
   &DynamicMaterialParameters(
    Index 0 Param 1=&FloatFromCurve(FloatCurve=DataInterface() {propertyValues {Curve {keys [<at 0.3 1 RCIM_Cubic>,<at 1 -0.1>]}}})
    Index 0 Param 2=&FloatFromCurve(FloatCurve=DataInterface() {propertyValues {Curve {keys [<at 0 1 RCIM_Cubic>,<at 0.8 8.8 RCIM_Cubic>]}}})
    Index 0 Param 4=&FloatFromCurve(FloatCurve=DataInterface() {propertyValues {Curve {keys [<at 0.3 0>,<at 0.5 0.5>]}}})
    Module.Param0WriteEnabled=true
    Module.Param1WriteEnabled=true
    Module.Param2WriteEnabled=true
    Module.Param3WriteEnabled=true
   )
   &CameraOffset(Camera Offset Amount=-5)
  ]
  rendererClasses [&NiagaraSpriteRendererProperties]
  renderers [
   {rendererClass=&NiagaraSpriteRendererProperties}
   {rendererClass=&NiagaraSpriteRendererProperties, rendererIndex=1}
  ]
 }
}
summary {
 systemName=NS_RoastingMarshmellows_Emote_Default
 emitters [
  {emitterName=SmokeCardsL001, rendererClasses [&NiagaraSpriteRendererProperties]}
  {bEnabled=false, emitterName=SmokeCardsL, rendererClasses [&NiagaraSpriteRendererProperties]}
 ]
 userVariables [
  User.bIsFrontEndPreview {type=NiagaraBool}
  User.IsSprinting {type=NiagaraBool}
  User.PlayerSpeed {type=NiagaraFloat, defaultValue=0, description=""}
  User.SkeletalMesh {
   type=NiagaraDataInterfaceSkeletalMesh
   defaultValue {
    struct=NiagaraExt_VariableValue_DataInterface
    value {dataInterface=&self:NiagaraDataInterfaceSkeletalMesh_0}
   }
  }
 ]
}
